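# Physics asset for the Forest Queen medium female player body, covering a dangling horn and knife at the hips rather than a full-body ragdoll.
# Box bodies cover the horn, horn base and knife bones, while capsules give the pelvis and both thighs collision volumes.
# Two constraints tie knife and horn directly to the pelvis with tight swing and twist limits.
# The small horn base box has no constraint.
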PhysicsAsset /BRCosmetics/Characters/Player/Female/Medium/Bodies/F_MED_Forest_Queen/Meshes/F_MED_Forest_Queen_Physics
bodies {
 dyn_horn [{shapeType=Box, extentX=11.2211, extentY=5.8192, extentZ=2.6819, center=<5.2111 0 0>}]
 dyn_horn_base [{shapeType=Box, extentX=0.75, extentY=0.75, extentZ=0.75}]
 dyn_knife [{shapeType=Box, extentX=17.2621, extentY=7.4214, extentZ=2.9355, center=<7.7603 0 0.0001>}]
 pelvis [{shapeType=Capsule, length=0.375, radius=12.375, center=<0.6592 -0.1608 -0.9964>}]
 thigh_l [{shapeType=Capsule, length=24.875, radius=7.125, center=<-6.3038 0.0003 1.6999>, rotation=<rot 77.551 -115.7709 -107.4651>}]
 thigh_r [{shapeType=Capsule, length=23.625, radius=7.125, center=<11.863 -0.0001 -0.1943>, rotation=<rot 83.2573 180 180>}]
}
constraints [
 {bone1Name=dyn_knife, bone2Name=pelvis, swing1LimitDegrees=20, swing2LimitDegrees=20, twistLimitDegrees=3}
 {bone1Name=dyn_horn, bone2Name=pelvis, swing1LimitDegrees=20, swing2LimitDegrees=8, twistLimitDegrees=4}
]
