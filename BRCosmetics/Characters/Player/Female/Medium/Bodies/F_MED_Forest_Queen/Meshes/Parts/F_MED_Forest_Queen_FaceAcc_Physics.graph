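# Physics asset for the Forest Queen female cosmetic's face accessory mesh, covering a hair braid and a right-side bang.
# The braid is a capsule chain from a small box base through parent and five segments, with swing and twist limits widest at the first segment.
# The bang is a short box chain held to tight limits.
# Unconstrained head, neck, upper spine and right forearm capsules supply body shapes alongside the hair.
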PhysicsAsset /BRCosmetics/Characters/Player/Female/Medium/Bodies/F_MED_Forest_Queen/Meshes/Parts/F_MED_Forest_Queen_FaceAcc_Physics
bodies {
 dyn_bang_r_01 [{shapeType=Box, extentX=2.4345, extentY=0.6425, extentZ=1.9681, center=<0.9603 0.3053 0.2973>}]
 dyn_bang_r_base [{shapeType=Box, extentX=0.5, extentY=0.5, extentZ=0.5}]
 dyn_bang_r_parent [{shapeType=Box, extentX=2.5122, extentY=0.8517, extentZ=2.1098, center=<1.0506 -0.0001 0>}]
 dyn_face_parent [{shapeType=Box, extentX=0.75, extentY=0.75, extentZ=0.75}]
 dyn_hair_braid_01 [{shapeType=Capsule, length=2.5302, radius=1.5986, center=<2.7804 0.1046 0>, rotation=<rot 90 0 -5.7905>}]
 dyn_hair_braid_02 [{shapeType=Capsule, length=2.3618, radius=2.1299, center=<2.8157 0.1648 0>, rotation=<rot 90 179.1048 175.8494>}]
 dyn_hair_braid_03 [{shapeType=Capsule, length=1.3679, radius=1.8664, center=<1.9811 0.5219 0>, rotation=<rot 90 0 -15.0392>}]
 dyn_hair_braid_04 [{shapeType=Capsule, length=1.0102, radius=1.8619, center=<1.6644 0.5524 0>, rotation=<rot 90 14.0363 -5.2196>}]
 dyn_hair_braid_05 [{shapeType=Capsule, length=6.0218, radius=2.3685, center=<4.8411 -0.2744 0>, rotation=<rot 90 0 0>}]
 dyn_hair_braid_base [{shapeType=Box, extentX=0.75, extentY=0.75, extentZ=0.75}]
 dyn_hair_braid_parent [{shapeType=Capsule, length=4.255, radius=0.5, center=<2.5514 -0.5967 0.3279>, rotation=<rot 74.8347 0.0001 10.1916>}]
 head [{shapeType=Capsule, length=2.375, radius=7.25, center=<2.961 0.3111 0>, rotation=<rot -90 0 0>}]
 lowerarm_r [{shapeType=Capsule, length=26.5, radius=5, center=<13.0603 -5.4192 0.0002>, rotation=<rot 90 26.565 53.055>}]
 neck_01 [{shapeType=Capsule, length=7.875, radius=5.375, center=<3.0403 -0.0029 -0.1402>, rotation=<rot 90 0 0>}]
 spine_04 [{shapeType=Capsule, length=12, radius=12, center=<3.348 -2.9562 2.4321>, rotation=<rot 87.9973 -179.9991 -148.9058>}]
 spine_05 [
  {shapeType=Capsule, length=15.375, radius=6.75, center=<5.765 -4.5739 -1.1107>}
  {shapeType=Capsule, length=9, radius=11.25, center=<-1.4155 -1.9763 -0.0001>, rotation=<rot 90 0 36.3226>}
  {shapeType=Capsule, length=7.75, radius=12.5, center=<-16.8406 -2.7058 -0.0001>, rotation=<rot 90 0 -1.6436>}
 ]
}
constraints [
 {bone1Name=dyn_hair_braid_01, bone2Name=dyn_hair_braid_parent, swing1LimitDegrees=45, swing2LimitDegrees=20, twistLimitDegrees=20}
 {bone1Name=dyn_hair_braid_02, bone2Name=dyn_hair_braid_01, swing1LimitDegrees=35, swing2LimitDegrees=15, twistLimitDegrees=20}
 {bone1Name=dyn_hair_braid_03, bone2Name=dyn_hair_braid_02, swing1LimitDegrees=20, swing2LimitDegrees=15, twistLimitDegrees=30}
 {bone1Name=dyn_hair_braid_04, bone2Name=dyn_hair_braid_03, swing1LimitDegrees=20, swing2LimitDegrees=15, twistLimitDegrees=10}
 {bone1Name=dyn_hair_braid_05, bone2Name=dyn_hair_braid_04, swing1LimitDegrees=20, swing2LimitDegrees=15, twistLimitDegrees=10}
 {bone1Name=dyn_hair_braid_parent, bone2Name=dyn_hair_braid_base, swing1LimitDegrees=20, swing2LimitDegrees=15, twistLimitDegrees=10}
 {bone1Name=dyn_bang_r_parent, bone2Name=dyn_bang_r_base, swing1LimitDegrees=4, swing2LimitDegrees=8, twistLimitDegrees=4}
 {bone1Name=dyn_bang_r_01, bone2Name=dyn_bang_r_parent, swing1LimitDegrees=4, swing2LimitDegrees=8, twistLimitDegrees=4}
]
